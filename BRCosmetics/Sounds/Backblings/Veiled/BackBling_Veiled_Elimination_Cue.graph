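# Sound cue for the Veiled back bling's elimination reaction sound in Battle Royale cosmetics.
# A random node picks one of three equally weighted Veiled waves, and a modulator varies its pitch at constant volume.
# It uses close-range pet attenuation, the Backblings sound class and the reactive backbling property sheet.
# Overridden concurrency allows one instance per owner, stopping the oldest when retriggered.
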SoundCue /BRCosmetics/Sounds/Backblings/Veiled/BackBling_Veiled_Elimination_Cue
root=Modulator_0
nodes [
 {id=Modulator_0, children [Random_0], properties {pitchMin=0.85, volumeMax=1, volumeMin=1}}
 {id=Random_0, children [WavePlayer_0,WavePlayer_1,WavePlayer_2], properties {weights [1,1,1]}}
 {id=WavePlayer_0, properties {soundWaveAssetPtr=&/BP_Veiled_1}}
 {id=WavePlayer_1, properties {soundWaveAssetPtr=&/BP_Veiled_2}}
 {id=WavePlayer_2, properties {soundWaveAssetPtr=&/BP_Veiled_3}}
]
properties {
 attenuationSettings=&Fort_Pets_Close_Athena_Att
 audioPropertiesSheet=/BRCosmetics/Sounds/Settings/PropertySheets/Reactive/APS_Cosmetic_Reactive_Backbling
 bOverrideConcurrency=true
 maxDistance=1800.0001
 soundClassObject=/Game/Sounds/FortSoundClassesAndModes/Fort_Backblings
 volumeMultiplier=1
 concurrencyOverrides {bLimitToOwner=true, maxCount=1, resolutionRule=StopOldest}
}
